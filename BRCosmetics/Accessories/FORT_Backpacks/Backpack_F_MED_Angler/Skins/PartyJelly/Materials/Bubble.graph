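# Translucent lit material for the bubble shell of the PartyJelly Angler backpack skin, with a watery, glowing look.
# Noise-distorted, panning water normals, projected triplanar-style in pre-skinned space, blend with the baked normal and warp UVs.
# Baked-normal Fresnel drives opacity, color, thin-film gradient and blurred scene color seen through an emissive layer with faked sun highlight and cubemap reflection.
# Masked sine waves along vertex normals form a world position offset wobble.
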Material /BRCosmetics/Accessories/FORT_Backpacks/Backpack_F_MED_Angler/Skins/PartyJelly/Materials/Bubble
graph {
 properties {
  bContactShadows=true
  bScreenSpaceReflections=true
  bUsedWithMeshDeformer=true
  bUsedWithSkeletalMesh=true
  blendMode=BLEND_Translucent
  pixelDepthOffsetMode=PDOM_Legacy
  translucencyLightingMode=TLM_VolumetricDirectional
  usedShadingModels=Default Lit
 }
 nodes [
  n1 = ScalarParameter() {defaultValue=-0.5, parameterName=Jellyfish Height}
  n2 = Saturate()
  n3 = ScalarParameter() {defaultValue=0.1, parameterName=WaterTiling}
  n4 = PreSkinnedPosition()
  n5 = Divide(A=@n4, B=ScalarParameter() {defaultValue=40, group=PanningEmissive, parameterName=Triplanner_ConstantScaleDivisor})
  n6 = VertexInterpolator(VS=@n5)
  n7 = Multiply(A=Time(), B=VectorParameter() {group=PanningEmissive, parameterName=Water Speed, defaultValue=<rgba 0 0 0.1 1>}.RGB)
  n8 = Time()
  n9 = VectorParameter() {group=PanningEmissive, parameterName=TriPlanner Speed, defaultValue=<rgba 0 0 0.2 1>}
  n10 = Add(A=Multiply(A=ScalarParameter() {defaultValue=15, parameterName=Distortion Tiling}, B=@n5), B=Frac(None=Multiply(A=@n8, B=@n9.RGB)))
  n11 = TextureObjectParameter() {group=PanningEmissive, parameterName=EmissiveGradient, texture=/Game/Global/Textures/Default/Noise/T_Tiling_Noise_01}
  n12 = LinearInterpolate(
   A=TextureSample(UVs=ComponentMask(None=@n10) {b=true, r=true}, Tex=@n11) {samplerType=SAMPLERTYPE_Masks, texture=/Game/Weapons/WeaponSkins/Textures/T_Swirly_Noise_highContrast}.RGB
   B=TextureSample(UVs=ComponentMask(None=@n10) {b=true, g=true}, Tex=@n11) {samplerType=SAMPLERTYPE_Masks, texture=&T_Swirly_Noise_highContrast}.RGB
  )
  n13 = Add(
   A=Add(A=Multiply(A=@n3, B=@n6.PS), B=Frac(None=@n7))
   B=Multiply(A=VertexInterpolator(VS=@n12).PS, B=ScalarParameter() {defaultValue=0.05, parameterName=Offset Intensity}) {constB=0.01}
  )
  n14 = TextureObjectParameter() {
   group=PanningEmissive
   parameterName=WaterNormal
   samplerType=SAMPLERTYPE_Normal
   texture=/Engine/Functions/Engine_MaterialFunctions02/ExampleContent/Textures/water_n
  }
  n15 = PreSkinnedNormal()
  n16 = MaterialFunctionCall(
   In=Abs(None=ComponentMask(None=@n15) {b=true})
   Contrast=ScalarParameter() {defaultValue=0.2, group=PanningEmissive, parameterName=TriPlanner_AxisBlendStrength}
  ) {materialFunction=&CheapContrast}
  n17 = LinearInterpolate(
   A=TextureSample(UVs=ComponentMask(None=@n13) {b=true, r=true}, Tex=@n14) {samplerType=SAMPLERTYPE_Normal, texture=&T_Swirly_Noise_highContrast}.RGB
   B=TextureSample(UVs=ComponentMask(None=@n13) {g=true, r=true}, Tex=@n14) {samplerType=SAMPLERTYPE_Normal, texture=&T_Swirly_Noise_highContrast}.RGB
   Alpha=VertexInterpolator(VS=ComponentMask(None=@n16.Result) {b=true}).PS
  )
  n18 = Multiply(A=ComponentMask(None=@n17) {g=true, r=true}, B=ScalarParameter() {defaultValue=0.01, parameterName=Emissive Warp})
  Add(A=BumpOffset(Coordinate=TextureCoordinate(), Height=LinearInterpolate(A=@n1, Alpha=@n2) {constB=0}), B=@n18)
  n19 = TextureSampleParameter2D() {
   parameterName=Baked Normal
   samplerType=SAMPLERTYPE_Normal
   texture=/BRCosmetics/Accessories/FORT_Backpacks/Backpack_F_MED_Angler/Texture/T_F_MED_Angler_Backpack_N
  }
  n20 = Fresnel(ExponentIn=ScalarParameter() {defaultValue=0.8131, parameterName=TranparencyFresnel}, Normal=Transform(None=@n19.RGB))
  LinearInterpolate(
   A=ScalarParameter() {defaultValue=0.7333, parameterName=Opacity Min}
   B=ScalarParameter() {defaultValue=1, parameterName=Opacity Max}
   Alpha=@n20
  )
  n21 = Saturate(None=OneMinus(None=@n20))
  n22 = Saturate(None=Power(Base=@n21, Exp=ScalarParameter() {defaultValue=1.2, parameterName=BlurFresnel}) {constExponent=4})
  LinearInterpolate(
   A=VectorParameter() {parameterName=Base Color, defaultValue=<rgba 0.0248 0.0233 0.0312 1>}.RGB
   B=VectorParameter() {parameterName=Inner Shell Color, defaultValue=<rgba 0.7166 0.9192 1 1>}.RGB
   Alpha=@n22
  )
  n23 = CollectionParameter() {
   collection=&FortniteMaterialParameters
   parameterName=SunAndMoonModelDirectionalVector
  }
  n24 = MaterialFunctionCall(Float3=@n23) {materialFunction=&BreakOutFloat3Components}
  n25 = MaterialFunctionCall(X=@n24.R, Y=@n24.R, Z=Constant() {r=1}) {materialFunction=&MakeFloat3}
  n26 = Multiply(
   A=DotProduct(A=ComponentMask(None=@n25.Result) {b=true, g=true, r=true}, B=ReflectionVectorWS())
   B=ScalarParameter() {defaultValue=0.8993, parameterName=SpecMult}
  )
  n27 = ScalarParameter() {defaultValue=70, parameterName=Spec Highlight Power}
  n28 = Multiply(A=Clamp(None=Power(Base=@n26, Exp=@n27) {constExponent=20}), B=ScalarParameter() {defaultValue=3, parameterName=Spec Intense})
  n29 = TextureSampleParameterCube!(UVs=ReflectionVectorWS() {bNormalizeCustomWorldNormal=true}, Bias=Constant() {r=3}, Apply View MipBias) {
   mipValueMode=TMVM_MipBias
   parameterName=CubeMap
   samplerSource=SSM_Wrap_WorldGroupSettings
   samplerType=SAMPLERTYPE_LinearColor
   texture=/Game/Packages/Fortress_Sky/TexturesHDR/T_AthenaSkylight
  }
  n30 = VectorParameter() {parameterName=Cube Map Multiplier, defaultValue=<rgba 0.22 0.22 0.5 0>}
  n31 = MaterialFunctionCall() {materialFunction=/Engine/Functions/GetPostProcessFunctions/GetAmbientCubemapTint}
  n32 = Desaturation(
   None=Multiply(A=Multiply(A=@n29.RGB, B=@n30.RGB) {constB=1.5}, B=ComponentMask(None=@n31.AmbientCubemapTint) {b=true, g=true, r=true})
   Fraction=Constant() {r=0.5}
  )
  n33 = VectorParameter() {parameterName=DeepSea Emissive, defaultValue=<rgba 0 6.178 12 1>}
  n34 = Add(
   A=Multiply(
    A=Multiply(A=@n33.RGB, B=@n22)
    B=TextureSample() {texture=/BRCosmetics/Accessories/FORT_Backpacks/Backpack_F_MED_Angler/Skins/PartyJelly/Textures/T_PartyJelly_Backpack_E}.RGB
   )
   B=Add(A=VectorParameter() {parameterName=Base Shell Color, defaultValue=<rgba 0.4042 0.4684 1 1>}.RGB, B=Add(A=@n28, B=@n32))
  )
  n35 = Multiply(
   A=VectorParameter() {parameterName=Fresnel emiss color, defaultValue=<rgba 0.0564 0.8222 0.94 0>}.RGB
   B=ScalarParameter() {defaultValue=1, parameterName=Fresnel emiss mult}
  )
  n36 = Multiply(A=Fresnel() {baseReflectFraction=0.005}, B=ScalarParameter() {defaultValue=1, parameterName=Fresnel emiss})
  n37 = Constant() {r=1}
  MaterialFunctionCall(Emissive=LinearInterpolate(
    A=LinearInterpolate(A=@n34, B=@n35, Alpha=@n36)
    B=MaterialFunctionCall!(Mask=@n37) {materialFunction=/BRCosmetics/Characters/Player/Common/SharedMaterials/SceneColor_Blur}.Result
    Alpha=@n22
   )) {materialFunction=&MF_FX_EmissiveAdjustments}
  n38 = TextureSampleParameter2D() {parameterName=WPO_Mask, samplerType=SAMPLERTYPE_Masks, texture=&T_WhiteMask}
  n39 = WorldPosition() {worldPositionShaderOffset=WPT_ExcludeAllShaderOffsets}
  n40 = Subtract(A=Frac(None=Multiply(A=@n39.XYZ) {constB=0.01})) {constB=0.5}
  n41 = MaterialFunctionCall(Float3=Abs(None=Multiply(A=@n40) {constB=2})) {materialFunction=&BreakOutFloat3Components}
  n42 = Add(A=@n41.R, B=@n41.R)
  n43 = Time()
  n44 = Subtract(A=Subtract(A=Add(A=@n42, B=@n41.B)) {constB=2}, B=Frac(None=Multiply(A=@n43) {constB=0.25}))
  n45 = Multiply(A=Add(A=Sine(None=@n44))) {constB=0.5}
  n46 = ScalarParameter() {defaultValue=1, parameterName=WPO_WorldNoiseScale}
  n47 = MaterialFunctionCall(Float3=Multiply() {constB=0.02}) {materialFunction=&BreakOutFloat3Components}
  n48 = Subtract(A=Add(A=Add(A=@n47.R, B=@n47.R), B=@n47.B)) {constB=2}
  n49 = Time()
  n50 = ScalarParameter() {parameterName=Slurp_WPO_random_seed}
  n51 = Frac(None=Multiply(A=Subtract(A=@n49, B=@n50)) {constB=0.6})
  n52 = Add(A=Sine(None=Subtract(A=@n48, B=@n51)))
  n53 = Multiply(A=VertexNormalWS(), B=Add(A=Multiply(A=@n45, B=@n46), B=Multiply(A=@n52) {constB=0.5}))
  n54 = ComponentMask(None=Saturate(None=MaterialFunctionCall() {materialFunction=/Engine/Functions/Engine_MaterialFunctions02/Texturing/3DSandMayaUVCoordinates}.UVChannel1)) {r=true}
  Multiply(
   A=Multiply(
    A=MaterialFunctionCall(Float3=@n38.RGB) {materialFunction=&BreakOutFloat3Components}.B
    B=Multiply(A=@n53, B=@n54)
   )
   B=ScalarParameter() {defaultValue=1, parameterName=WPO_FinalScale}
  )
  n55 = ScalarParameter() {defaultValue=0.6, parameterName=WaterIntensity}
  Reroute(None=MaterialFunctionCall(
    BaseNormal=@n19.RGB
    AdditionalNormal=MaterialFunctionCall(Normal=@n17, Flatness=@n55) {materialFunction=&FlattenNormal}.Result
   ) {materialFunction=&BlendAngleCorrectedNormals}.Result)
  n56 = LinearInterpolate(Alpha=ComponentMask(None=@n17) {g=true}) {constA=1.5, constB=3}
  n57 = TextureSample(UVs=Add(A=Multiply(A=@n56, B=@n20) {constA=2}, B=@n18)) {texture=/BRCosmetics/Characters/Player/Female/Medium/Bodies/F_MED_Dragon_Racer/Textures/Blue/Textures/T_Gradient_Dragon_Racer_Color_Blue}
  n58 = Multiply(A=VertexInterpolator(VS=@n12).PS, B=@n21)
  Reroute(None=Multiply(A=Multiply(A=@n57.RGB, B=@n58), B=ScalarParameter() {defaultValue=1, parameterName=ThinFilm Intensity}))
  ScalarParameter() {defaultValue=0.181, parameterName=Roughness}
  ScalarParameter() {defaultValue=0.75, parameterName=Specular}
  VertexInterpolator(VS=@n4)
  VertexInterpolator(VS=@n15)
 ]
}
